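# Anim Blueprint for the QuaintToy backpack that nudges its pack bone to follow the wearer's body motion.
# Each update reads spine-to-pelvis yaw, adds a skydive offset, and interpolates pack roll and translation.
# Left upper-arm yaw drives extra roll, weighted by a smoothed melee-animation alpha.
# The AnimGraph stacks four additive Transform Bone nodes on the input pose, mixing fixed offsets with these driven values.
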AnimBlueprint /BRCosmetics/Accessories/FORT_Backpacks/Backpack_F_MED_QuaintToy/Meshes/F_MED_QuaintToy_Pack_AnimBP
target_skeleton=&/F_MED_QuaintToy_Pack_Skeleton
graphs [
 AnimGraph {nodes [
   Component To Local()
   Component To Local()
   n1 = "Transform (Modify) Bone - Bone: pack"(ComponentPose=Local To Component(LocalPose=Input Pose()), Translation="0,-0.5,0", Rotation="0,0,0") {
    rotationMode=BMM_Additive
    translationMode=BMM_Additive
    translationSpace=BCS_BoneSpace
    alphaBoolBlend {blendOption=Linear}
    boneToModify {boneName=pack}
    translation {y=-0.5}
   }
   n2 = "Transform (Modify) Bone - Bone: pack"!(ComponentPose=@n1, Alpha, Translation=Get backpack_translation(), Rotation=Get backpack_rotate()) -> (Pose) {
    rotationMode=BMM_Additive
    rotationSpace=BCS_BoneSpace
    translationMode=BMM_Additive
    translationSpace=BCS_BoneSpace
    boneToModify {boneName=pack}
   }
   n3 = "Transform (Modify) Bone - Bone: pack"!(
    ComponentPose=@n2
    Alpha=Get isMelee() -> (isMelee:double)
    Translation="0,-1.5,0"
    Rotation=Get Backpack Rotate Pickaxe() -> (Backpack Rotate Pickaxe:Rotator)
    Scale:Vector
   ) -> (Pose) {
    rotationMode=BMM_Additive
    rotationSpace=BCS_BoneSpace
    translationMode=BMM_Additive
    translationSpace=BCS_BoneSpace
    alphaBoolBlend {blendOption=Linear}
    alphaScaleBiasClamp {interpSpeedDecreasing=5, interpSpeedIncreasing=5}
    boneToModify {boneName=pack}
    rotation {roll=20}
    translation {y=-1.5}
   }
   n4 = "Transform (Modify) Bone - Bone: pack"!(ComponentPose=@n3, bAlphaBoolEnabled:bool, Translation="0,-1,0", Rotation="0,0,4", Scale) -> (Pose) {
    alphaInputType=Bool
    rotationMode=BMM_Additive
    rotationSpace=BCS_BoneSpace
    translationMode=BMM_Additive
    translationSpace=BCS_BoneSpace
    alphaBoolBlend {blendInTime=0.2, blendOption=Linear, blendOutTime=0.2}
    alphaScaleBiasClamp {interpSpeedDecreasing=5, interpSpeedIncreasing=5}
    boneToModify {boneName=pack}
    rotation {roll=4}
    translation {y=-1}
   }
   Output Pose(Result=Component To Local(ComponentPose=@n4)) {layerGroup=None}
  ]}
 EventGraph {nodes [
   n1 = Event BlueprintUpdateAnimation()
   n2 = GetOwnerSkeletalMeshComponent()
   n3 = BreakRotator(InRot=BreakTransform(InTransform=GetDeltaTransformFromRefPose(self=@n2, BoneName=spine_05, BaseName=pelvis)).Rotation)
   n4 = float + float(A=MapRangeClamped(Value=@n3.Yaw, InRangeA=0, InRangeB=-30, OutRangeB=35), B=Get SkyDive Additive() -> (SkyDive Additive:double))
   > Set backpack_rotate(backpack_rotate=RInterpTo!(Current:Rotator=Get backpack_rotate(), Target:Rotator=MakeRotator(Roll=@n4), DeltaTime:float=@n1.DeltaTimeX, InterpSpeed:float=10) -> (ReturnValue:Rotator))
   n5 = float + float(A=MapRangeClamped(Value=@n3.Yaw, InRangeA=0, InRangeB=-30, OutRangeB=-10), B=Get SkyDive Additive())
   > Set backpack_translation(backpack_translation=VInterpTo!(Current:Vector=Get backpack_translation(), Target:Vector=MakeVector(Y=@n5), DeltaTime:float=@n1.DeltaTimeX, InterpSpeed:float=10) -> (ReturnValue:Vector))
   n6 = Get bIsPlayingMeleeAnim()
   > Set isMelee(isMelee:double=FInterpTo(
     Current=Get isMelee()
     Target="To Float (Boolean)"(InBool=@n6)
     DeltaTime=float * float(A=@n1.DeltaTimeX, B=0.2) {is=title K2Node_PromotableOperator}
     InterpSpeed=5
    )) -> (Output_Get:double)
   n7 = GetOwnerSkeletalMeshComponent()
   n8 = BreakRotator(InRot=BreakTransform(InTransform=GetDeltaTransformFromRefPose(self=@n7, BoneName=upperarm_l, BaseName=clavicle_l)).Rotation)
   > Set Backpack Rotate Pickaxe(Backpack Rotate Pickaxe:Rotator=MakeRotator(Roll=MapRangeClamped(Value=@n8.Yaw, InRangeA=0, InRangeB=45, OutRangeB=20))) -> (Output_Get:Rotator)
   Event OnBeginSkydiving()
   > Set SkyDive Additive(SkyDive Additive:double=-3) -> (Output_Get:double)
   Event OnEndSkydiving()
   > Set SkyDive Additive(SkyDive Additive=0)
   TryGetPawnOwner()
   float * float(
    A=Set head down rot(head down rot:double=GetAlphaFromReferenceTransformDelta(BoneName=spine_01, BaseName=head, InputRange=(X=0,Y=-45), TwistAxis="1,0,0")) -> (Output_Get:double)
    B=Reroute Node(InputPin:double) -> (OutputPin:double)
   )
  ]}
]
variables [
 Backpack Rotate Pickaxe {type=Rotator}
 backpack_rotate {type=Rotator}
 backpack_translation {type=Vector}
 head down rot {type=number}
 isMelee {type=number}
 SkyDive Additive {type=number}
]
